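# M_Primitive_Glit is a two-sided translucent base material using the Thin Translucent shading model, usable on skeletal, instanced, Niagara and Geometry Collection meshes.
# MF_Primitive_Glit supplies the TransmittanceColor that feeds the Thin Translucent output node.
# It is set up for material attributes, index-of-refraction refraction and per-pixel surface lighting.
# High-quality forward reflections, per-pixel fog, responsive AA and translucent velocity output are enabled.
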Material /AtomFramework/Materials/Game_Qual/Base/M_Primitive_Glit
graph {
 properties {
  bComputeFogPerPixel=true
  bEnableResponsiveAA=true
  bOutputTranslucentVelocity=true
  bUseHQForwardReflections=true
  bUseMaterialAttributes=true
  bUseTranslucencyVertexFog=false
  bUsedWithGeometryCollections=true
  bUsedWithInstancedStaticMeshes=true
  bUsedWithMeshDeformer=true
  bUsedWithNiagaraMeshParticles=true
  bUsedWithSkeletalMesh=true
  blendMode=BLEND_Translucent
  pixelDepthOffsetMode=PDOM_Legacy
  refractionMethod=RM_IndexOfRefraction
  shadingModel=MSM_ThinTranslucent
  translucencyLightingMode=TLM_SurfacePerPixelLighting
  twoSided=true
  usedShadingModels=Thin Translucent
 }
 nodes [ThinTranslucentMaterialOutput!(
   TransmittanceColor=MaterialFunctionCall() {materialFunction=/AtomFramework/Materials/MatFunctions/MF_Primitive_Glit}.TransmittanceColor
   SurfaceCoverage
  )]
}
