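# M_Primitive_Opal is a two-sided translucent base material for an opal surface, using the Thin Translucent shading model.
# The MF_Primitive_Opal material function supplies TransmittanceColor to the Thin Translucent output, the only connection shown.
# It uses per-pixel surface lighting, per-pixel fog and index-of-refraction refraction, with contact and dynamic masked shadows but no ray-traced shadows.
# Usage flags cover Nanite, instanced static meshes, geometry collections and Niagara mesh particles.
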Material /AtomFramework/Materials/Game_Qual/Base/M_Primitive_Opal
graph {
 properties {
  bCastDynamicShadowAsMasked=true
  bCastRayTracedShadows=false
  bComputeFogPerPixel=true
  bContactShadows=true
  bEnableResponsiveAA=true
  bOutputTranslucentVelocity=true
  bUseHQForwardReflections=true
  bUseMaterialAttributes=true
  bUseTranslucencyVertexFog=false
  bUsedWithGeometryCollections=true
  bUsedWithInstancedStaticMeshes=true
  bUsedWithNanite=true
  bUsedWithNiagaraMeshParticles=true
  blendMode=BLEND_Translucent
  pixelDepthOffsetMode=PDOM_Legacy
  refractionMethod=RM_IndexOfRefraction
  shadingModel=MSM_ThinTranslucent
  translucencyLightingMode=TLM_SurfacePerPixelLighting
  twoSided=true
  usedShadingModels=Thin Translucent
 }
 nodes [ThinTranslucentMaterialOutput!(
   TransmittanceColor=MaterialFunctionCall() {materialFunction=/AtomFramework/Materials/MatFunctions/MF_Primitive_Opal}.TransmittanceColor
   SurfaceCoverage
  )]
}
