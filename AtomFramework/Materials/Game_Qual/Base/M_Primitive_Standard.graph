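# M_Primitive_Standard is a base material for primitives that uses the Subsurface shading model.
# It works through Material Attributes, with its graph holding a single call to the MF_Primitive_Standard material function, where the shading logic lives.
# Usage flags cover Nanite, skeletal and instanced static meshes, Niagara mesh particles, Geometry Collections and Mesh Deformer.
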Material /AtomFramework/Materials/Game_Qual/Base/M_Primitive_Standard
graph {
 properties {
  bUseMaterialAttributes=true
  bUsedWithGeometryCollections=true
  bUsedWithInstancedStaticMeshes=true
  bUsedWithMeshDeformer=true
  bUsedWithNanite=true
  bUsedWithNiagaraMeshParticles=true
  bUsedWithSkeletalMesh=true
  pixelDepthOffsetMode=PDOM_Legacy
  shadingModel=MSM_Subsurface
  usedShadingModels=Subsurface
 }
 nodes [MaterialFunctionCall() {materialFunction=/AtomFramework/Materials/MatFunctions/MF_Primitive_Standard}]
}
