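# M_Primitive_Trans is a two-sided translucent base material using the Thin Translucent shading model, with per-pixel surface lighting and screen-space reflections.
# It uses material attributes and index-of-refraction refraction, and is flagged for Nanite, skeletal, instanced, geometry collection, mesh deformer and Niagara mesh use.
# The visible graph delegates to MF_Primitive_Trans, whose TransmittanceColor output feeds the Thin Translucent output node.
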Material /AtomFramework/Materials/Game_Qual/Base/M_Primitive_Trans
graph {
 properties {
  bScreenSpaceReflections=true
  bUseMaterialAttributes=true
  bUsedWithGeometryCollections=true
  bUsedWithInstancedStaticMeshes=true
  bUsedWithMeshDeformer=true
  bUsedWithNanite=true
  bUsedWithNiagaraMeshParticles=true
  bUsedWithSkeletalMesh=true
  blendMode=BLEND_Translucent
  pixelDepthOffsetMode=PDOM_Legacy
  refractionMethod=RM_IndexOfRefraction
  shadingModel=MSM_ThinTranslucent
  translucencyLightingMode=TLM_SurfacePerPixelLighting
  twoSided=true
  usedShadingModels=Thin Translucent
 }
 nodes [ThinTranslucentMaterialOutput!(
   TransmittanceColor=MaterialFunctionCall() {materialFunction=/AtomFramework/Materials/MatFunctions/MF_Primitive_Trans}.TransmittanceColor
   SurfaceCoverage
  )]
}
