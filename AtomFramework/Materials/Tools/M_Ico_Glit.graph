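# M_Ico_Glit is a UI-domain material in AtomFramework's material tools, apparently for an icon.
# It uses the blue channel of the MatIcons_M texture as a mask to blend from a Col color parameter, white by default, toward a constant grey.
# The graph shows no connection from this blend to any material output.
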Material /AtomFramework/Materials/Tools/M_Ico_Glit
graph {
 properties {materialDomain=MD_UI, pixelDepthOffsetMode=PDOM_Legacy, usedShadingModels=Default Lit}
 nodes [LinearInterpolate(
   A=VectorParameter() {parameterName=Col, defaultValue=<rgba 1 1 1 0>}.RGB
   Alpha=TextureSample() {samplerType=SAMPLERTYPE_LinearColor, texture=&/MatIcons_M}.B
  ) {constB=0.4}]
}
