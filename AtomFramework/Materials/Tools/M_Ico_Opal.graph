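# UI-domain material that colors an opal icon, going by its name, from the green channel of the MatIcons_M texture.
# That channel blends blue to purple, and its remapped absolute distance from mid-gray blends the Col parameter toward that gradient.
# Mid-gray texels therefore show Col (white by default), while dark texels turn blue and bright texels turn purple.
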Material /AtomFramework/Materials/Tools/M_Ico_Opal
graph {
 properties {materialDomain=MD_UI, pixelDepthOffsetMode=PDOM_Legacy, usedShadingModels=Default Lit}
 nodes [
  n1 = TextureSample() {samplerType=SAMPLERTYPE_LinearColor, texture=&/MatIcons_M}
  LinearInterpolate(
   A=VectorParameter() {parameterName=Col, defaultValue=<rgba 1 1 1 0>}.RGB
   B=LinearInterpolate(A=Constant3Vector() {constant=<rgba 0.0181 0.3269 1 1>}, B=Constant3Vector() {constant=<rgba 0.6606 0 1 1>}, Alpha=@n1.G)
   Alpha=Abs(None=ConstantBiasScale(None=@n1.G) {bias=-0.5, scale=2})
  )
 ]
}
