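# M_Ico_Trans is a User Interface domain material in the AtomFramework tools folder that builds a tintable color from an icon texture.
# It samples the red channel of MatIcons_M as linear color and adds a constant offset to brighten it.
# That value multiplies the RGB of the Col vector parameter, which defaults to white.
# The graph shows no opacity logic despite the Trans name.
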Material /AtomFramework/Materials/Tools/M_Ico_Trans
graph {
 properties {materialDomain=MD_UI, pixelDepthOffsetMode=PDOM_Legacy, usedShadingModels=Default Lit}
 nodes [Multiply(
   A=VectorParameter() {parameterName=Col, defaultValue=<rgba 1 1 1 0>}.RGB
   B=Add(A=TextureSample() {samplerType=SAMPLERTYPE_LinearColor, texture=&/MatIcons_M}.R) {constB=0.6}
  )]
}
