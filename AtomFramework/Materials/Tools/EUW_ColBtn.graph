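# EUW_ColBtn is a small editor utility swatch button, apparently one colour entry in the EUW_ColSet material tool.
# PreConstruct tints the button with Colour, labels it with ID and builds its style from an icon material selected by MatType.
# Only Transparent, Glitter and Opalescent have icon materials; other types get no brush resource.
# Clicking passes ID to Set Material_Event on MasterRef, leaving the material change to the parent tool.
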WidgetBlueprint /AtomFramework/Materials/Tools/EUW_ColBtn
graphs [EventGraph {nodes [
   "On Clicked (Button_27)"() {is=title K2Node_ComponentBoundEvent}
   > Set Material_Event!(
    execute
    self:EUW Col Set Object Reference=Get MasterRef() -> (MasterRef:EUW Col Set Object Reference)
    NewParam:Name=Get ID() -> (ID:Name=None)
   ) -> (then) {is=title K2Node_CallFunction}
   Event PreConstruct()
   > SetBackgroundColor!(
    execute
    self:Button Object Reference=Get Button_27() -> (Button_27:Button Object Reference)
    InBackgroundColor:Linear Color Structure=Get Colour() -> (Colour:Linear Color Structure)
   ) -> (then)
   > "SetText (Text)"(
    self=Get TextBlock_40() -> (TextBlock_40:Text Object Reference)
    InText="To Text (Name)"!(InName:Name=Get ID() -> (ID=None)) -> (ReturnValue:Text)
   )
   n1 = Make Slate Brush!(
    TintColor:Slate Color Structure=(SpecifiedColor=(R=1,G=1,B=1,A=1),ColorUseRule=UseColor_Specified)
    DrawAs:ESlateBrushDrawType Enum=Image
    Tiling:ESlateBrushTileType Enum=NoTile
    Mirroring:ESlateBrushMirrorType Enum=NoMirror
    ImageSize:Vector 2D Structure=(X=32,Y=32)
    Margin:Margin Structure=(Left=0,Top=0,Right=0,Bottom=0)
    ResourceObject:Object Reference=Select!(
     None:Object Reference
     Standard:Object Reference
     Transparent:Object Reference=&/M_Ico_Trans
     Glitter:Object Reference=&/M_Ico_Glit
     Opalescent:Object Reference=&/M_Ico_Opal
     Metallic:Object Reference
     Index:EAtomMaterialType Enum=Get MatType() -> (MatType:EAtomMaterialType Enum=None)
    ) -> (ReturnValue:Object Reference)
    OutlineSettings:Slate Brush Outline Settings Structure=(CornerRadii=(X=0,Y=0,Z=0,W=0),Color=(SpecifiedColor=(R=0,G=0,B=0,A=0),ColorUseRule=UseColor_Specified),Width=0,RoundingType=HalfHeightRadius,bUseBrushTransparency=False)
   ) -> (SlateBrush:Slate Brush Structure) {is=title K2Node_MakeStruct}
   > SetStyle!(
    execute
    self:Button Object Reference=Get Button_27()
    InStyle:"Button Style Structure (by ref)"=Make Button Style!(
     Normal:Slate Brush Structure=@n1
     Hovered:Slate Brush Structure=@n1
     Pressed:Slate Brush Structure=@n1
     Disabled:Slate Brush Structure=(TintColor=(SpecifiedColor=(R=1,G=1,B=1,A=1),ColorUseRule=UseColor_Specified),DrawAs=NoDrawType,Tiling=NoTile,Mirroring=NoMirror,ImageType=NoImage,ImageSize=(X=0,Y=0),Margin=(Left=0,Top=0,Right=0,Bottom=0),ResourceObject=None,OutlineSettings=(CornerRadii=(X=0,Y=0,Z=0,W=1),Color=(SpecifiedColor=(R=0,G=0,B=0,A=0),ColorUseRule=UseColor_Specified),Width=0,RoundingType=HalfHeightRadius,bUseBrushTransparency=False),UVRegion=(Min=(X=0,Y=0),Max=(X=0,Y=0),bIsValid=False),bIsDynamicallyLoaded=False,ResourceName="")
     NormalForeground:Slate Color Structure=(SpecifiedColor=(R=1,G=0,B=1,A=1),ColorUseRule=UseColor_Foreground)
     HoveredForeground:Slate Color Structure=(SpecifiedColor=(R=1,G=0,B=1,A=1),ColorUseRule=UseColor_Foreground)
     PressedForeground:Slate Color Structure=(SpecifiedColor=(R=1,G=0,B=1,A=1),ColorUseRule=UseColor_Foreground)
     DisabledForeground:Slate Color Structure=(SpecifiedColor=(R=1,G=0,B=1,A=1),ColorUseRule=UseColor_Foreground)
     NormalPadding:Margin Structure=(Left=0,Top=0,Right=0,Bottom=0)
     PressedPadding:Margin Structure=(Left=0,Top=0,Right=0,Bottom=0)
     PressedSlateSound:Slate Sound Structure=(ResourceObject=None)
     ClickedSlateSound:Slate Sound Structure=(ResourceObject=None)
     HoveredSlateSound:Slate Sound Structure=(ResourceObject=None)
    ) -> (ButtonStyle:Button Style Structure) {is=title K2Node_MakeStruct}
   ) -> (then)
  ]}]
variables [
 Colour {type=LinearColor, default=<rgba 1 1 1 1>}
 ID {type=string, default=None}
 MasterRef {type=EUW_ColSet_C, default=null}
 MatType {type=EAtomMaterialType, default=None}
]
widgets {
 info {inheritedWidgetCount=0, parentClass=&EditorUtilityWidget, rootWidgetClass=&SizeBox, widgetCount=3}
 widgets [
  SizeBox_42 {
   widgetClassPath=&SizeBox
   widget=SizeBox_42
   properties {bOverride_HeightOverride=true, bOverride_WidthOverride=true, heightOverride=35, widthOverride=35}
  }
  Button_27 {
   widgetClassPath=&Button
   parent=SizeBox_42
   widget=Button_27
   properties {backgroundColor=<rgba 0 1 0.3785 1>}
   slot {class=SizeBoxSlot}
  }
  TextBlock_40 {
   widgetClassPath=&TextBlock
   bIsVariable=true
   parent=Button_27
   widget=TextBlock_40
   properties {
    justification=Center
    text=103
    font {
     bForceMonospaced=false
     bMaterialIsStencil=false
     fontMaterial=null
     fontObject=&Roboto
     letterSpacing=0
     monospacedWidth=1
     size=10
     skewAmount=0
     typefaceFontName=Bold
     outlineSettings {
      bApplyOutlineToDropShadows=false
      bMiteredCorners=false
      bSeparateFillAlpha=false
      outlineMaterial=null
      outlineSize=0
      outlineColor=<rgba 0 0 0 1>
     }
    }
    shadowColorAndOpacity=<rgba 0 0 0 1>
   }
   slot {class=ButtonSlot}
  }
 ]
}
