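# Post-process material, applied before depth of field, that redraws the scene as a monochrome sketch with dark edges and diagonal screen-space hatching.
# Edges come from world-normal differences at four diagonal offsets scaled by Edge Size, plus fixed-offset depth differences.
# Dividing scene color by diffuse color approximates lighting, which fades the banded hatching to white in bright areas.
# An optional stencil mask limits the effect to unoccluded custom-depth pixels matching Stencil Target, blended by Stencil Fade.
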Material /AtomFramework/Materials/Tools/PP_EdgeShader
graph {
 properties {blendableLocation=BL_SceneColorBeforeDOF, materialDomain=MD_PostProcess, usedShadingModels=Default Lit}
 nodes [
  n1 = Constant()
  n2 = If(
   A=ComponentMask(None=SceneTexture() {sceneTextureId=PPI_CustomStencil}.Color) {r=true}
   B=ScalarParameter() {defaultValue=1, parameterName=Stencil Target}
   "A > B"=@n1
   "A == B"=Constant() {r=1}
   "A < B"=@n1
  )
  n3 = Constant() {r=1}
  n4 = If(
   A=SceneTexture() {sceneTextureId=PPI_CustomDepth}.Color
   B=SceneTexture() {sceneTextureId=PPI_SceneDepth}.Color
   "A > B"=Constant()
   "A == B"=@n3
   "A < B"=@n3
  )
  NamedRerouteDeclaration(None=Multiply(A=Multiply(A=@n2, B=@n4), B=ScalarParameter() {defaultValue=1, parameterName=Stencil Fade})) {name=Stencil Shape, nodeColor=<rgba 0.5937 1 0 1>}
  n5 = SceneTexture() {bFiltered=true, sceneTextureId=PPI_WorldNormal}
  n6 = TextureCoordinate()
  n7 = Constant2Vector() {g=1, r=1}
  n8 = ScalarParameter() {defaultValue=0.001, parameterName=Edge Size, sliderMax=0.001}
  n9 = SceneTexture(UVs=Add(A=@n6, B=Multiply(A=@n7, B=@n8))) {bFiltered=true, sceneTextureId=PPI_WorldNormal}
  n10 = Constant2Vector() {g=-1, r=-1}
  n11 = SceneTexture(UVs=Add(A=@n6, B=Multiply(A=@n10, B=@n8))) {bFiltered=true, sceneTextureId=PPI_WorldNormal}
  n12 = Constant2Vector() {g=-1, r=1}
  n13 = SceneTexture(UVs=Add(A=@n6, B=Multiply(A=@n12, B=@n8))) {bFiltered=true, sceneTextureId=PPI_WorldNormal}
  n14 = Constant2Vector() {g=1, r=-1}
  n15 = SceneTexture(UVs=Add(A=@n6, B=Multiply(A=@n14, B=@n8))) {bFiltered=true, sceneTextureId=PPI_WorldNormal}
  n16 = Max(
   A=Max(A=Distance(A=@n5.Color, B=@n9.Color), B=Distance(A=@n5.Color, B=@n11.Color))
   B=Max(A=Distance(A=@n5.Color, B=@n13.Color), B=Distance(A=@n5.Color, B=@n15.Color))
  )
  n17 = SceneTexture() {bFiltered=true, sceneTextureId=PPI_SceneDepth}
  n18 = TextureCoordinate()
  n19 = Constant() {r=0.001}
  n20 = Distance(A=@n17.Color, B=SceneTexture(UVs=Add(A=@n18, B=@n19)) {bFiltered=true, sceneTextureId=PPI_SceneDepth}.Color)
  n21 = TextureCoordinate()
  n22 = Constant() {r=-0.001}
  n23 = Distance(A=@n17.Color, B=SceneTexture(UVs=Add(A=@n21, B=@n22)) {bFiltered=true, sceneTextureId=PPI_SceneDepth}.Color)
  n24 = Saturate(None=Divide(A=Max(A=@n20, B=@n23)) {constB=100})
  n25 = ScreenPosition()
  n26 = Frac(None=DotProduct(A=Divide(A=@n25.PixelPosition) {constB=5}, B=Constant2Vector() {g=0.5, r=0.5}))
  n27 = Abs(None=Multiply(A=Subtract(A=@n26) {constB=0.5}) {constB=2})
  n28 = ScalarParameter() {defaultValue=0.2, parameterName=Band Offset, sliderMax=0.5, sliderMin=-0.5}
  n29 = Add(A=Constant() {r=0.49}, B=@n28)
  n30 = Add(A=Constant() {r=0.51}, B=@n28)
  n31 = Multiply(A=Saturate(None=MaterialFunctionCall(Input=@n27, Input Low=@n29, Input High=@n30) {materialFunction=&RemapValueRange}.Result)) {constB=0.5}
  n32 = SceneTexture() {bFiltered=true, sceneTextureId=PPI_PostProcessInput0}
  n33 = SceneTexture() {bFiltered=true, sceneTextureId=PPI_DiffuseColor}
  n34 = Power(Base=Desaturation(None=Divide(A=@n32.Color, B=@n33.Color)) {luminanceFactors=<rgba 1 1 1 1>}) {constExponent=7}
  n35 = Multiply(
   A=Multiply(A=OneMinus(None=@n16), B=OneMinus(None=@n24))
   B=LinearInterpolate(A=Add(A=@n31) {constB=0.25}, B=Constant() {r=1}, Alpha=Saturate(None=@n34))
  )
  StaticSwitchParameter(True=LinearInterpolate(A=SceneTexture() {sceneTextureId=PPI_PostProcessInput0}.Color, B=@n35, Alpha=NamedRerouteUsage()), False=@n35) {parameterName=Enable Stencil Mask}
 ]
}
